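; Dance floor launch pad trap that launches players into skydive after someone on it emotes.
; Overlapping pawns bind to emote-start events; an emote starts a countdown that launches every tagged occupant with gameplay cues and events.
; The pad then stays active for a timed window, instantly launching entrants, before resetting to idle.
; OnRep_LaunchActive updates Niagara and material visuals; leftover test-implementation events remain, some empty.
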(event EventBeginPlay
  (Networking|SwitchHasAuthority
    (:Authority
      (CallFunction|OnBecomeIdle))))

(event EventOnPlaced
  (Trap|AddTriggerComponent (Variables|Default|GetCylinderTriggerArea)))

(event Custom|OnLaunchBecomeActive
  (Development|PrintString "Dance floor launch active")
  (bind _returnvalue (Utilities|Time|SetTimerbyEvent (EventDispatchers|CreateEvent :!event Custom|OnBecomeIdle) (Variables|Default|GetLaunchActiveTime)))
  (Variables|Default|SetIdleTimerHandle _returnvalue)
  (Variables|Default|SetLaunchActive true)
  (Niagara|SetNiagaraVariable(LinearColor) (Variables|Default|GetNS_DanceFloorLaunchPad) "Color" (Variables|Default|GetActiveColor))
  (Rendering|Material|SetScalarParameterValueonMaterials (Variables|Default|GetTriggerAreaVisualizer) "ActiveIdleValue" 1.0))

(event Custom|CheckForDancingPawns
  ; Comment
  (CallFunction|ClearLaunchVariables)
  (for _array_element (Collision|GetOverlappingActors (Variables|Default|GetCylinderTriggerArea) (Variables|Default|GetFortPlayerPawn) :!class /Script/Engine.PrimitiveComponent)
    (bind _asfort_player_pawn (Utilities|Casting|CastToFortPlayerPawn _array_element))
    (Utilities|Array|AddUnique (Variables|Default|GetPlayersToLaunch) _asfort_player_pawn)
    (if (GameplayTags|HasAnyMatchingGameplayTags (Utilities|Casting|ToGameplayTagAssetInterface _array_element) (Variables|Default|GetDanceTag))
      (Variables|Default|SetDancingPawnFound true)))
  (if (Variables|Default|GetDancingPawnFound)
    (Utilities|Time|ClearandInvalidateTimerbyHandle (Variables|Default|GetActiveLaunchTimerHandle))
    (CallFunction|OnLaunchBecomeActive)))

(event Custom|OnBecomeIdle
  (Development|PrintString "Dance floor idle")
  (CallFunction|ClearLaunchVariables)
  (Variables|Default|SetLaunchActive false)
  (Niagara|SetNiagaraVariable(LinearColor) (Variables|Default|GetNS_DanceFloorLaunchPad) "Color" (Variables|Default|GetIdleColor))
  (Rendering|Material|SetScalarParameterValueonMaterials (Variables|Default|GetTriggerAreaVisualizer) "ActiveIdleValue" 0.0))

(event OnComponentBeginOverlap(CylinderTriggerArea)
  (params (OverlappedComponent :type Primitive%20Component%20Object%20Reference) (OtherActor :type Actor%20Object%20Reference) (OtherComp :type Primitive%20Component%20Object%20Reference) (OtherBodyIndex :type Integer) (bFromSweep :type Boolean) (SweepResult :type Hit%20Result%20Structure%20%28by%20ref%29))
  (bind _asfort_player_pawn (Utilities|Casting|CastToFortPlayerPawn OtherActor))
  (if (Variables|Default|GetLaunchActive)
    (CallFunction|InstantLaunch _asfort_player_pawn)
    (else
      (EventDispatchers|BindEventtoOnPawnStartedEmote _asfort_player_pawn (delegate Custom|DancingPlayerDetected))
      (Utilities|Array|Add (Variables|Default|GetPlayersToLaunch) _asfort_player_pawn)
      (CallFunction|CheckLateOverlaps))))

(event Custom|DancingPlayerDetected
  (params (MontageItemDef :type Fort%20Item%20Definition%20Object%20Reference) (PawnEmoting :type Fort%20Pawn%20Object%20Reference))
  (Development|PrintString "SOMEONE IS DANCING ON THE DANCE FLOOR")
  (if (Variables|Default|GetNotAlreadyLaunching)
    (CallFunction|StartLaunchCountdown)
    (GameplayCue|AddGameplayCue self (Variables|Default|GetGC_CountdownLoop))))

(event OnComponentEndOverlap(CylinderTriggerArea)
  (params (OverlappedComponent :type Primitive%20Component%20Object%20Reference) (OtherActor :type Actor%20Object%20Reference) (OtherComp :type Primitive%20Component%20Object%20Reference) (OtherBodyIndex :type Integer))
  (bind _asfort_player_pawn (Utilities|Casting|CastToFortPlayerPawn OtherActor))
  (if (not (Variables|Default|GetLaunchActive))
    (EventDispatchers|UnbindEventfromOnPawnStartedEmote _asfort_player_pawn (EventDispatchers|CreateEvent :!event Custom|DancingPlayerDetected))
    (Utilities|Array|RemoveItem (Variables|Default|GetPlayersToLaunch) _asfort_player_pawn)))

(event Custom|FoundPawn
  ; Test Implementation 2
  (bind _returnvalue (Utilities|Time|SetTimerbyEvent (delegate Custom|LaunchPlayerOuterCircle) (Variables|Default|GetLaunchCountdown)))
  (Variables|Default|SetCurrentTimerHandle _returnvalue)
  (Utilities|Map|Add (Variables|Default|GetPlayerTimerMap) (Variables|Default|GetCurrentPlayerPawn) (Variables|Default|GetCurrentTimerHandle))
  (Networking|SwitchHasAuthority
    (:Remote
      (Niagara|SetNiagaraVariable(LinearColor) (Variables|Default|GetNS_DanceFloorLaunchPad) "Color" (Variables|Default|GetActiveColor))
      (Rendering|Material|SetScalarParameterValueonMaterials (Variables|Default|GetTriggerAreaVisualizer) "ActiveIdleValue" 1.0))))

(event Custom|LaunchPlayerOuterCircle
  ; Test Implementation 2
  (CallFunction|LaunchPlayer false))

(event Custom|StartLaunchCountdown
  (bind _launchvelocity (Variables|Default|GetLaunchVelocity))
  (Variables|Default|SetNotAlreadyLaunching false)
  (CallFunction|CheckLateOverlaps)
  (CallFunction|PlayWarmupAnimation)
  (Utilities|FlowControl|Delay (Variables|Default|GetLaunchCountdown))
  (CallFunction|OnLaunchBecomeActive)
  (GameplayCue|RemoveGameplayCue self (Variables|Default|GetGC_CountdownLoop))
  (for _array_element (Variables|Default|GetPlayersToLaunch)
    (EventDispatchers|UnbindEventfromOnPawnStartedEmote _array_element (EventDispatchers|CreateEvent :!event Custom|DancingPlayerDetected))
    (if (GameplayTags|HasAnyMatchingGameplayTags _array_element (Variables|Default|GetCanLaunchTagContainer))
      (AthenaMovement|SetPendingSkydiveLaunch _array_element true)
      (Character|LaunchCharacter _array_element _launchvelocity true true)
      (GameplayCue|ExecuteGameplayCue _array_element (Variables|Default|GetGC_PlayerLaunch))
      (Ability|SendGameplayEventtoActor _array_element "(TagName=\"Gameplay.Action.Environment.UtopiaCity.DanceFloor.PlayerLaunched\")" (Utilities|Struct|MakeGameplayEventData "(TagName=\"Gameplay.Action.Environment.UtopiaCity.DanceFloor.PlayerLaunched\")" 0 _array_element 0 0 "()" "(GameplayTags=)" (GameplayTags|GetOwnedGameplayTags _array_element :!class /Script/GameplayTags.BlueprintGameplayTagLibrary) (.z _launchvelocity)))))
  (Utilities|Array|Clear (Variables|Default|GetPlayersToLaunch)))

(event Custom|InstantLaunch
  (params (InPlayerPawn :type Fort%20Player%20Pawn%20Object%20Reference))
  (bind _launchvelocity (Variables|Default|GetLaunchVelocity))
  (if (GameplayTags|HasAnyMatchingGameplayTags InPlayerPawn (Variables|Default|GetCanLaunchTagContainer))
    (AthenaMovement|SetPendingSkydiveLaunch InPlayerPawn true)
    (Character|LaunchCharacter InPlayerPawn _launchvelocity true true)
    (GameplayCue|ExecuteGameplayCue InPlayerPawn (Variables|Default|GetGC_PlayerLaunch))
    (Ability|SendGameplayEventtoActor InPlayerPawn "(TagName=\"Gameplay.Action.Environment.UtopiaCity.DanceFloor.PlayerLaunched\")" (Utilities|Struct|MakeGameplayEventData "(TagName=\"Gameplay.Action.Environment.UtopiaCity.DanceFloor.PlayerLaunched\")" 0 InPlayerPawn 0 0 "()" "(GameplayTags=)" (GameplayTags|GetOwnedGameplayTags InPlayerPawn :!class /Script/GameplayTags.BlueprintGameplayTagLibrary) (.z _launchvelocity)))))

(event Custom|LaunchPlayer
  (params (CenterRing :type Boolean)))

(event Custom|CheckForPawnsTimer
  ; Test Implementation 2
  (CallFunction|CheckforPawns))

(event Custom|CheckForPawns)

(fn ShouldTrigger
  (params (TouchingActors :type Array%20of%20Actor%20Object%20References))
  (return false))

(fn CheckLateOverlaps
  (for _array_element (Collision|GetOverlappingActors (Variables|Default|GetCylinderTriggerArea) (Variables|Default|GetFortPlayerPawn) :!class /Script/Engine.PrimitiveComponent)
    (bind _asfort_player_pawn (Utilities|Casting|CastToFortPlayerPawn _array_element))
    (if (not (Utilities|Array|ContainsItem (Variables|Default|GetPlayersToLaunch) _asfort_player_pawn))
      (Utilities|Array|Add (Variables|Default|GetPlayersToLaunch) _asfort_player_pawn)
      (if (not (Variables|Default|GetLaunchActive))
        (EventDispatchers|BindEventtoOnPawnStartedEmote _asfort_player_pawn (EventDispatchers|CreateEvent :!event DancingPlayerDetected)))))
  (return))

(fn ClearLaunchVariables
  (Variables|Default|SetNotAlreadyLaunching true)
  (Utilities|Array|Clear (Variables|Default|GetPlayersToLaunch))
  (return))

(fn OnRep_LaunchActive
  (Rendering|Material|SetColorParameterValueonMaterials (Variables|Default|GetTriggerAreaVisualizer) "TopColor" (Math|Color|SelectColor (Variables|Default|GetActiveColor) (Variables|Default|GetIdleColor) (Variables|Default|GetLaunchActive)))
  (Rendering|Material|SetScalarParameterValueonMaterials (Variables|Default|GetTriggerAreaVisualizer) "ActiveIdleValue" (Math|Float|SelectFloat 1.0 0.0 (Variables|Default|GetLaunchActive)))
  (Niagara|SetNiagaraVariable(LinearColor) (Variables|Default|GetNS_DanceFloorLaunchPad) "Color" (Math|Color|SelectColor (Variables|Default|GetActiveColor) (Variables|Default|GetIdleColor) (Variables|Default|GetLaunchActive)))
  (if (Variables|Default|GetLaunchActive)
    (bind _asmaterial_instance_dynamic (Utilities|Casting|CastToMaterialInstanceDynamic (Rendering|Material|GetMaterial (Variables|Default|GetSM_UtopiaCity_Club_DanceFloor_GPad_A_A) 0))
      (:then
        (Rendering|Material|SetScalarParameterValue _asmaterial_instance_dynamic "TriggerTime" (Utilities|Time|GetGameTimeinSeconds) :!class /Script/Engine.MaterialInstanceDynamic)
        (Rendering|Material|SetScalarParameterValue _asmaterial_instance_dynamic "TriggerDuration" (Variables|Default|GetLaunchActiveTime) :!class /Script/Engine.MaterialInstanceDynamic))
      (:CastFailed
        _)))
  (return))

(fn PlayWarmupAnimation
  (bind _asmaterial_instance_dynamic (Utilities|Casting|CastToMaterialInstanceDynamic (Rendering|Material|GetMaterial (Variables|Default|GetSM_UtopiaCity_Club_DanceFloor_GPad_A_A) 0))
    (:then
      (Rendering|Material|SetScalarParameterValue _asmaterial_instance_dynamic "TriggerTime_Warmup" (Utilities|Time|GetGameTimeinSeconds) :!class /Script/Engine.MaterialInstanceDynamic)
      (Rendering|Material|SetScalarParameterValue _asmaterial_instance_dynamic "TriggerDuration_Warmup" (Variables|Default|GetLaunchCountdown) :!class /Script/Engine.MaterialInstanceDynamic)
      (return))
    (:CastFailed
      (return))))

(fn ConstructionScript
  (|CallParentFunction)
  (bind _returnvalue (Rendering|Material|CreateDynamicMaterialInstance (Variables|Default|GetSM_UtopiaCity_Club_DanceFloor_GPad_A_A) 0 (Rendering|Material|GetMaterial (Variables|Default|GetSM_UtopiaCity_Club_DanceFloor_GPad_A_A) 0) "DanceFloorMID" :!class /Script/Engine.PrimitiveComponent))
  (Rendering|Material|SetMaterial (Variables|Default|GetSM_UtopiaCity_Club_DanceFloor_GPad_A_A) 0 _returnvalue))
