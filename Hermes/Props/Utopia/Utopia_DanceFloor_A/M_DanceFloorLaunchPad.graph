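# Translucent, unlit, two-sided glow material for the Utopia dance floor launch pad's cylinder.
# Emissive blends vertical gradients and a noise-driven dot-grid hot base between idle and active colors, boosted by a timestamped flash and tinted by particle color.
# Opacity math combines UV masks, axis-aligned angle, depth and near-camera fades, and time-of-day dimming.
# Quality and platform switches simplify or scale effects, and vertex color channels feed named reroutes.
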Material /Hermes/Props/Utopia/Utopia_DanceFloor_A/M_DanceFloorLaunchPad
graph {
 properties {
  bAutomaticallySetUsageInEditor=false
  blendMode=BLEND_Translucent
  previewMesh=/Hermes/Props/Utopia/Utopia_DanceFloor_A/Mesh/SM_Cylinder_Floor_Pad_Glow
  shadingModel=MSM_Unlit
  twoSided=true
  usedShadingModels=Unlit
 }
 nodes [
  n1 = Divide(
   A=Subtract(A=Time() {period=5}, B=ScalarParameter() {parameterName=FlashStartTime})
   B=ScalarParameter() {defaultValue=0.75, group=Timestamp, parameterName=FlashDuration}
  )
  n2 = Multiply(
   A=MaterialFunctionCall!(In=Clamp(None=@n1)) {materialFunction=/Game/Effects/Fort_Effects/Materials/_Functions/MF_Smoothstep_EaseOut}.Result
   B=ScalarParameter() {defaultValue=4, parameterName=FlashBrightness}
  )
  n3 = ParticleColor()
  n4 = ScalarParameter() {defaultValue=0.2, parameterName=FresnelBaseBrightness}
  n5 = NamedRerouteUsage()
  n6 = LinearInterpolate(
   A=QualitySwitch(Default=Fresnel(BaseReflectFractionIn=@n4) {exponent=1.5}, Low=Constant() {r=0.3})
   B=ScalarParameter() {defaultValue=0.6, parameterName=SplineBrightness}
   Alpha=@n5
  )
  n7 = ComponentMask(None=TextureCoordinate()) {g=true}
  n8 = Power(Base=@n7, Exp=ScalarParameter() {defaultValue=1, parameterName=GradientPower})
  n9 = LinearInterpolate(
   A=VectorParameter() {parameterName=TopColor, defaultValue=<rgba 1.6425 0.2125 2 0>}.RGB
   B=VectorParameter() {parameterName=BrightColor, defaultValue=<rgba 0.7032 0.1497 2 0>}.RGB
   Alpha=Multiply(A=@n6, B=@n8)
  )
  n10 = OneMinus(None=@n5)
  n11 = TextureCoordinate()
  n12 = Constant2Vector() {g=88, r=256}
  n13 = TextureCoordinate() {uTiling=4, vTiling=4}
  n14 = MaterialFunctionCall(
   UVs=Frac(None=Multiply(A=@n11, B=@n12))
   Radius=Constant() {r=0.2}
   Density=TextureSample(UVs=Panner(Coordinate=@n13) {speedY=0.2}) {samplerType=SAMPLERTYPE_LinearColor, texture=/Game/Global/Textures/Default/Tileable_Patterns/T_PerlinNoise_Uncompressed}.R
  ) {materialFunction=&RadialGradientExponential}
  n15 = Fresnel(BaseReflectFractionIn=@n4) {exponent=1.5}
  n16 = Power(Base=Add(A=@n10, B=Multiply(A=@n14.RadialGradientExponential, B=@n15)), Exp=ScalarParameter() {defaultValue=10, parameterName=HotBasePower})
  n17 = ScalarParameter() {parameterName=ActiveIdleValue}
  n18 = LinearInterpolate(
   A=VectorParameter() {parameterName=HotBaseColorIdle, defaultValue=<rgba 3.1162 0.3906 5 0>}.RGB
   B=VectorParameter() {parameterName=HotBaseColorActive, defaultValue=<rgba 0.3906 3.4901 5 0>}.RGB
   Alpha=@n17
  )
  n19 = LinearInterpolate(
   A=VectorParameter() {parameterName=TopColorActive, defaultValue=<rgba 0.2125 0.4052 2 0>}.RGB
   B=VectorParameter() {parameterName=BrightColorActive, defaultValue=<rgba 0.1497 0.8948 2 0>}.RGB
  )
  n20 = LinearInterpolate(
   A=Add(A=@n9, B=Multiply(A=@n16, B=@n18))
   B=Multiply(A=LinearInterpolate(A=@n9, B=@n19, Alpha=@n17), B=OneMinus(None=@n16))
   Alpha=ScalarParameter() {parameterName=FadeOutBottom}
  )
  MaterialFunctionCall(Emissive=Multiply(A=Add(A=@n2), B=Multiply(A=@n3.RGB, B=@n20))) {materialFunction=&MF_FX_EmissiveAdjustments}
  n21 = ScalarParameter() {defaultValue=0.88, parameterName=MorningEveningBrightness}
  n22 = MaterialFunctionCall(
   Color_Morning=@n21
   Color_Day=Constant() {r=1}
   Color_Evening=@n21
   Color_Night=ScalarParameter() {defaultValue=0.6, parameterName=NightBrightness}
  ) {materialFunction=&Color_Time_of_Day}
  n23 = LinearInterpolate(A=ScalarParameter() {defaultValue=0.25, parameterName=InnerBrightness}, Alpha=ConstantBiasScale(None=TwoSidedSign())) {constA=0.25}
  n24 = VectorParameter() {parameterName=AngleFadeAxis, defaultValue=<rgba 0 0 1 1>}
  n25 = ScalarParameter() {defaultValue=1, parameterName=AngleFadeExp}
  n26 = Multiply(A=VertexNormalWS(), B=TwoSidedSign())
  n27 = Multiply(
   A=Power(
    Base=MaterialFunctionCall!(Axis=@n24.RGB, Exponent=@n25, Normal=@n26) {materialFunction=/Game/Effects/Fort_Effects/Materials/_Functions/MF_AxisAlignedFrensnel}.Fresnel
    Exp=ScalarParameter() {defaultValue=3, parameterName=AngleFadePower}
   )
   B=ScalarParameter() {defaultValue=5, parameterName=AngleFadeMult}
  )
  n28 = NamedRerouteUsage()
  n29 = NamedRerouteUsage()
  n30 = Clamp(None=QualitySwitch(Default=LinearInterpolate(A=@n27, Alpha=@n28), Low=Constant() {r=1}), Min=@n29)
  n31 = LinearInterpolate(
   A=ScalarParameter() {defaultValue=40, parameterName=MaxDepthFade}
   B=Multiply(A=ScalarParameter() {defaultValue=150, parameterName=SplineDepthFade}, B=@n29)
   Alpha=@n8
  )
  n32 = ScalarParameter() {defaultValue=200, parameterName=NearFadeDistance}
  n33 = Multiply(
   A=@n6
   B=Multiply(
    A=DepthFade(Opacity=@n30, FadeDistance=@n31)
    B=MaterialFunctionCall(Distance=@n32) {materialFunction=&Near_Camera_Fade}.Result
   )
  )
  n34 = MaterialFunctionCall(Speed=Constant() {r=0.5}) {materialFunction=&DebugTimeSine}
  n35 = Multiply(
   A=Multiply(A=Multiply(A=@n7, B=@n33), B=LinearInterpolate(Alpha=@n34.Result) {constA=0.4})
   B=ScalarParameter() {defaultValue=0.64, parameterName=OpacityMult}
  )
  n36 = ComponentMask(None=TextureCoordinate()) {r=true}
  n37 = Constant() {r=0.5}
  n38 = OneMinus(None=NamedRerouteUsage())
  n39 = Power(Base=Multiply(A=SphereMask(A=@n36, B=@n37, Radius=@n37, Hardness=@n38), B=NamedRerouteUsage())) {constExponent=1.4}
  Multiply(A=@n3.A, B=Multiply(A=Multiply(A=@n22.Result, B=@n23), B=Multiply(A=@n35, B=@n39)))
  Multiply(A=@n10)
  n40 = Constant() {r=1}
  n41 = Constant() {r=0.5}
  n42 = TextureCoordinate()
  n43 = Multiply(A=Power(Base=ComponentMask(None=@n42) {g=true}) {constExponent=4}) {constB=0.05}
  n44 = MaterialFunctionCall(Distance=ScalarParameter() {defaultValue=200, parameterName=NearFadeDistance}) {materialFunction=&Near_Camera_Fade}
  n45 = Multiply(
   A=Multiply(A=Multiply(A=@n43, B=@n44.Result), B=@n27)
   B=DepthFade(Opacity=Constant() {r=1}, FadeDistance=ScalarParameter() {parameterName=FinalPass-DepthFade-Distance})
  )
  n46 = ShadingPathSwitch(Default=@n45, Deferred=@n45) {desc=Migrated from FeatureLevelSwitch}
  n47 = Fresnel(ExponentIn=ScalarParameter() {defaultValue=1, parameterName=EdgeFresnelExp}, Normal=@n26) {baseReflectFraction=0.2, exponent=1.5}
  Multiply(
   A=MaterialFunctionCall("Xb1,PS4"=@n40, "PC-High&&Epic,XSX,PS5"=@n40, "PC-Perf,Android"=@n41, Switch=@n41, "PC-Low(NotPerf)&&Medium"=@n41) {materialFunction=&PlatformSwitch_1D}.Result
   B=LinearInterpolate(A=@n46, B=Multiply(A=@n46, B=@n47), Alpha=ScalarParameter() {parameterName=UseEdgeFresnel})
  )
  n48 = VertexColor()
  NamedRerouteDeclaration(None=@n48.R) {name=Vertex_Red, nodeColor=<rgba 0 1 0.3906 1>}
  NamedRerouteDeclaration(None=@n48.G) {name=Vertex Green, nodeColor=<rgba 1 0.0703 0 1>}
  NamedRerouteDeclaration(None=@n48.B) {name=Vertex Blue, nodeColor=<rgba 0 0.25 1 1>}
  OneMinus(None=@n47)
  Saturate(None=AppendVector(A=Divide(A=@n12.R, B=@n12.R), B=Divide(A=@n12.R, B=@n12.R)))
  VertexColor()
 ]
}
