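# Niagara effect system for the launch pad on the Utopia Dance Floor A prop, going by its name and folder.
# It holds a single Em_Particulates emitter, for which no renderers are listed, and the only listed module dependency is SystemState.
# A User.Color parameter, defaulting to magenta-purple, is exposed on the system.
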NiagaraSystem /Hermes/Props/Utopia/Utopia_DanceFloor_A/NS_DanceFloorLaunchPad
dependencies {usedDataInterfaces [], usedDynamicInputs [], usedModules [&SystemState]}
emitters {Em_Particulates {rendererClasses [], renderers []}}
summary {
 systemName=NS_DanceFloorLaunchPad
 emitters [{emitterName=Em_Particulates, rendererClasses []}]
 userVariables [User.Color {type=LinearColor, description="", defaultValue {struct=LinearColor, value=<rgba 0.5698 0.1074 0.5471 1>}}]
}
